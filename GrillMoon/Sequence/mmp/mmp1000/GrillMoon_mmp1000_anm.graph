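# Level Sequence for the animation layer of GrillMoon shot mmp1000, animating the MistMylar countdown mannequin and a mask gadget.
# Three successive body animation sections drive the attached mannequin, the last two overlapping, after it settles vertically while a separate MystMylar scale actor grows slightly.
# The attached skeletal mask gadget animates alongside the first body section, then visibility keys swap it for an attached static mask mesh as its animation ends.
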LevelSequence /GrillMoon/Sequence/mmp/mmp1000/GrillMoon_mmp1000_anm
bindings [
 Mask_GrillMoon_Gadget {
  id=#1
  possessed_class=FortSkeletalMeshActor
  tracks [
   {type=MovieSceneSpawnTrack, sections [{type=MovieSceneSpawnSection, channels [None {type=MovieSceneScriptingBoolChannel, default=true}]}]}
   {
    type=MovieScene3DTransformTrack
    sections [{
      type=MovieScene3DTransformSection
      channels [
       Location.X {type=MovieSceneScriptingDoubleChannel, default=0}
       Location.Y {type=MovieSceneScriptingDoubleChannel, default=0}
       Location.Z {type=MovieSceneScriptingDoubleChannel, default=0}
       Rotation.X {type=MovieSceneScriptingDoubleChannel, default=0}
       Rotation.Y {type=MovieSceneScriptingDoubleChannel, default=0}
       Rotation.Z {type=MovieSceneScriptingDoubleChannel, default=0}
       Scale.X {type=MovieSceneScriptingDoubleChannel, default=1}
       Scale.Y {type=MovieSceneScriptingDoubleChannel, default=1}
       Scale.Z {type=MovieSceneScriptingDoubleChannel, default=1}
      ]
     }]
   }
   {type=MovieSceneSkeletalAnimationTrack, sections [{type=MovieSceneSkeletalAnimationSection, end_frame=4698, start_frame=3660}]}
   {type=MovieScene3DAttachTrack, sections [{type=MovieScene3DAttachSection}]}
   {
    type=MovieSceneVisibilityTrack
    sections [{
      type=MovieSceneVisibilitySection
      channels [None {type=MovieSceneScriptingBoolChannel, default=true, keys [[3059,false],[3660,true],[4698,false]]}]
     }]
   }
  ]
 }
 _MystMylar_Scale {
  id=#2
  possessed_class=FortStaticMeshActor
  tracks [
   {type=MovieSceneSpawnTrack, sections [{type=MovieSceneSpawnSection, channels [None {type=MovieSceneScriptingBoolChannel, default=true}]}]}
   {
    type=MovieScene3DTransformTrack
    sections [{
      type=MovieScene3DTransformSection
      channels [
       Location.X {type=MovieSceneScriptingDoubleChannel, default=-57594.2109, const [-57594.2109,2]}
       Location.Y {type=MovieSceneScriptingDoubleChannel, default=42656.1797, const [42656.1797,2]}
       Location.Z {type=MovieSceneScriptingDoubleChannel, default=9936.5957, const [9936.5957,2]}
       Rotation.X {type=MovieSceneScriptingDoubleChannel, default=0, const [0,2]}
       Rotation.Y {type=MovieSceneScriptingDoubleChannel, default=0, const [0,2]}
       Rotation.Z {type=MovieSceneScriptingDoubleChannel, default=0, const [0,2]}
       Scale.X {type=MovieSceneScriptingDoubleChannel, default=1, keys [[2980,1],[3028,1.25]]}
       Scale.Y {type=MovieSceneScriptingDoubleChannel, default=1, keys [[2980,1],[3028,1.25]]}
       Scale.Z {type=MovieSceneScriptingDoubleChannel, default=1, keys [[2980,1],[3028,1.25]]}
      ]
     }]
   }
  ]
 }
 GrillMoon_Countdown_MistMylar {
  id=#3
  possessed_class=FortPlayerMannequin
  tracks [
   {type=MovieScene3DAttachTrack, sections [{type=MovieScene3DAttachSection}]}
   {
    type=MovieScene3DTransformTrack
    sections [{
      type=MovieScene3DTransformSection
      channels [
       Location.X {type=MovieSceneScriptingDoubleChannel, default=0, const [0,2]}
       Location.Y {type=MovieSceneScriptingDoubleChannel, default=0, const [0,2]}
       Location.Z {type=MovieSceneScriptingDoubleChannel, default=6.2331, keys [[2903,6.2331],[3028,0]]}
       Rotation.X {type=MovieSceneScriptingDoubleChannel, default=0, const [0,2]}
       Rotation.Y {type=MovieSceneScriptingDoubleChannel, default=0, const [0,2]}
       Rotation.Z {type=MovieSceneScriptingDoubleChannel, default=0, const [0,2]}
       Scale.X {type=MovieSceneScriptingDoubleChannel, default=1, const [1,2]}
       Scale.Y {type=MovieSceneScriptingDoubleChannel, default=1, const [1,2]}
       Scale.Z {type=MovieSceneScriptingDoubleChannel, default=1, const [1,2]}
      ]
     }]
   }
   {
    type=MovieSceneSkeletalAnimationTrack
    sections [
     {type=MovieSceneSkeletalAnimationSection, end_frame=5176, start_frame=3660}
     {type=MovieSceneSkeletalAnimationSection, end_frame=6388, start_frame=5176}
     {type=MovieSceneSkeletalAnimationSection, end_frame=7160, start_frame=6334}
    ]
   }
  ]
 }
 Face {
  id=#4
  parent_id=#3
  possessed_class=SkeletalMeshComponentBudgeted
  tracks [{
    type=MovieScene3DTransformTrack
    sections [{
      type=MovieScene3DTransformSection
      channels [
       Location.X {type=MovieSceneScriptingDoubleChannel, default=0}
       Location.Y {type=MovieSceneScriptingDoubleChannel, default=0}
       Location.Z {type=MovieSceneScriptingDoubleChannel, default=0}
       Rotation.X {type=MovieSceneScriptingDoubleChannel, default=0}
       Rotation.Y {type=MovieSceneScriptingDoubleChannel, default=0}
       Rotation.Z {type=MovieSceneScriptingDoubleChannel, default=0}
       Scale.X {type=MovieSceneScriptingDoubleChannel, default=1}
       Scale.Y {type=MovieSceneScriptingDoubleChannel, default=1}
       Scale.Z {type=MovieSceneScriptingDoubleChannel, default=1}
      ]
     }]
   }]
 }
 GrillMoon_Countdown_SM_mmMask {
  id=#5
  possessed_class=FortStaticMeshActor
  tracks [
   {type=MovieSceneSpawnTrack, sections [{type=MovieSceneSpawnSection, channels [None {type=MovieSceneScriptingBoolChannel, default=true}]}]}
   {type=MovieScene3DAttachTrack, sections [{type=MovieScene3DAttachSection}]}
   {
    type=MovieSceneVisibilityTrack
    sections [{
      type=MovieSceneVisibilitySection
      channels [None {type=MovieSceneScriptingBoolChannel, default=false, keys [[4690.9975,false],[4691.9975,true]]}]
     }]
   }
   {
    type=MovieScene3DTransformTrack
    sections [{
      type=MovieScene3DTransformSection
      channels [
       Location.X {type=MovieSceneScriptingDoubleChannel, default=-151.9145, const [-151.9145,1]}
       Location.Y {type=MovieSceneScriptingDoubleChannel, default=1.7393, const [1.7393,1]}
       Location.Z {type=MovieSceneScriptingDoubleChannel, default=0, const [0,1]}
       Rotation.X {type=MovieSceneScriptingDoubleChannel, default=179.9472, const [0.0396,1]}
       Rotation.Y {type=MovieSceneScriptingDoubleChannel, default=-90, const [-90,1]}
       Rotation.Z {type=MovieSceneScriptingDoubleChannel, default=-179.9472, const [-0.0396,1]}
       Scale.X {type=MovieSceneScriptingDoubleChannel, default=1, const [1,1]}
       Scale.Y {type=MovieSceneScriptingDoubleChannel, default=1, const [1,1]}
       Scale.Z {type=MovieSceneScriptingDoubleChannel, default=1, const [1,1]}
      ]
     }]
   }
  ]
 }
]
display_rate {denominator=1, numerator=60}
tick_resolution {denominator=1, numerator=24000}
tracks []
