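# GrillMoon_mmp1000_aud is the audio Level Sequence for the GrillMoon mmp1000 sequence.
# A master audio track named Masks holds one timed section at default volume and pitch.
# An event track on the Audio component of the BP_GrillMoon_MistMylar_Orb binding holds a single trigger.
# The Mask_GrillMoon_Gadget skeletal mesh is bound without tracks.
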LevelSequence /GrillMoon/Sequence/mmp/mmp1000/GrillMoon_mmp1000_aud
bindings [
 Mask_GrillMoon_Gadget {id=#1, possessed_class=FortSkeletalMeshActor, tracks []}
 Audio {
  id=#2
  parent_id=#3
  possessed_class=AudioComponent
  tracks [{
    type=MovieSceneEventTrack
    sections [{
      type=MovieSceneEventTriggerSection
      channels [None {type=MovieSceneScriptingEventChannel, const ["<Struct 'MovieSceneEvent' (0x850243bda20) {}>", 1]}]
     }]
   }]
 }
 BP_GrillMoon_MistMylar_Orb {id=#3, possessed_class=BP_MystMylar_Orb_C, tracks []}
]
display_rate {denominator=1, numerator=60}
tick_resolution {denominator=1, numerator=24000}
tracks [{
  type=MovieSceneAudioTrack
  display_name=Masks
  sections [{
    type=MovieSceneAudioSection
    end_frame=4845
    start_frame=3589
    channels [
     Volume {type=MovieSceneScriptingFloatChannel, default=1}
     Pitch {type=MovieSceneScriptingFloatChannel, default=1}
     AttachActor {type=MovieSceneScriptingActorReferenceChannel}
    ]
   }]
 }]
