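# GrillMoon mmp1000 Level Sequence providing camera shake, force feedback and screen effects for a cinematic moment.
# A spawned shared cameraForce Blueprint uses a demon-walk shake and keyed feedback effects (charge build-up, bumps, missile launch), fired through Lobby and Radius event tracks, with a dense Lobby feedback burst.
# A spawned PostProcessVolume pulses exposure and warm tint in time with the three last feedback hits, adding vignette and an intensity setting on the final one.
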LevelSequence /GrillMoon/Sequence/mmp/mmp1000/GrillMoon_mmp1000_cam
bindings [
 Events_Shared_cameraForce_BP {
  id=#1
  possessed_class=Events_Shared_cameraForce_BP_C
  tracks [
   {
    type=MovieSceneObjectPropertyTrack
    display_name=Force Effect
    sections [{
      type=MovieSceneObjectPropertySection
      channels [None {
        type=MovieSceneScriptingObjectPathChannel
        default="<Object '/Game/Weapons/ForceFeedback/FF_MissleLaunch' (0x85098d48ce0) Class 'ForceFeedbackEffect'>"
        keys [
         [0, "<Object '/Game/Weapons/ForceFeedback/FF_ATVBump' (0x84c4dc669e0) Class 'ForceFeedbackEffect'>"]
         [3721, "<Object '/GrillMoon/ForcedFeedback/GrillMoon_FF_Jim_Charge_BuildUp' (0x85098d48c00) Class 'ForceFeedbackEffect'>"]
         [4176, "<Object '&FF_ATVBump' (0x84c4dc669e0) Class 'ForceFeedbackEffect'>"]
         [4311, "<Object '&FF_MissleLaunch' (0x85098d48ce0) Class 'ForceFeedbackEffect'>"]
         [4627, "<Object '&FF_ATVBump' (0x84c4dc669e0) Class 'ForceFeedbackEffect'>"]
         [4770, "<Object '&Events_Shared_small_FF' (0x84c3c703480) Class 'ForceFeedbackEffect'>"]
        ]
       }]
     }]
   }
   {
    type=MovieSceneObjectPropertyTrack
    display_name=Cam Shake Base
    sections [{
      type=MovieSceneObjectPropertySection
      channels [None {
        type=MovieSceneScriptingObjectPathChannel
        default="<Object '&Events_Shared_00small_csb_C' (0x84c3c6a8a00) Class 'BlueprintGeneratedClass'>"
        const [
         "<Object '/GrillMoon/CameraShake/GrillMoon_pjw1000_demonwalk_csb.GrillMoon_pjw1000_demonwalk_csb_C' (0x850a101c600) Class 'BlueprintGeneratedClass'>"
         1
        ]
       }]
     }]
   }
   {
    type=MovieSceneEventTrack
    display_name=Cam Shake Radius Event Trigger
    sections [{
      type=MovieSceneEventTriggerSection
      channels [None {type=MovieSceneScriptingEventChannel, const ["<Struct 'MovieSceneEvent' (0x850243b1a40) {}>", 1]}]
     }]
   }
   {type=MovieSceneSpawnTrack, sections [{type=MovieSceneSpawnSection, channels [None {type=MovieSceneScriptingBoolChannel, default=true}]}]}
   {
    type=MovieSceneEventTrack
    display_name=Force Feedback Radius Event Trigger
    sections [{
      type=MovieSceneEventTriggerSection
      channels [None {type=MovieSceneScriptingEventChannel, const ["<Struct 'MovieSceneEvent' (0x850243b4600) {}>", 1]}]
     }]
   }
   {
    type=MovieScene3DTransformTrack
    sections [{
      type=MovieScene3DTransformSection
      channels [
       Location.X {type=MovieSceneScriptingDoubleChannel, default=-57728.7912}
       Location.Y {type=MovieSceneScriptingDoubleChannel, default=41385.2584}
       Location.Z {type=MovieSceneScriptingDoubleChannel, default=10992.7078}
       Rotation.X {type=MovieSceneScriptingDoubleChannel, default=0}
       Rotation.Y {type=MovieSceneScriptingDoubleChannel, default=0}
       Rotation.Z {type=MovieSceneScriptingDoubleChannel, default=0}
       Scale.X {type=MovieSceneScriptingDoubleChannel, default=1}
       Scale.Y {type=MovieSceneScriptingDoubleChannel, default=1}
       Scale.Z {type=MovieSceneScriptingDoubleChannel, default=1}
      ]
     }]
   }
   {
    type=MovieSceneIntegerTrack
    sections [{type=MovieSceneIntegerSection, channels [None {type=MovieSceneScriptingIntegerChannel, default=1, const [1,1]}]}]
   }
   {
    type=MovieSceneEventTrack
    display_name=Cam Shake Lobby Event Trigger
    sections [{
      type=MovieSceneEventTriggerSection
      channels [None {type=MovieSceneScriptingEventChannel, const ["<Struct 'MovieSceneEvent' (0x850243b8b60) {}>", 1]}]
     }]
   }
   {
    type=MovieSceneEventTrack
    display_name=Force Feedback Lobby Event Trigger
    sections [{
      type=MovieSceneEventTriggerSection
      channels [None {
        type=MovieSceneScriptingEventChannel
        keys [
         [3721, "<Struct 'MovieSceneEvent' (0x850243b6180) {}>"]
         [4176, "<Struct 'MovieSceneEvent' (0x850243bd660) {}>"]
         [4205, "<Struct 'MovieSceneEvent' (0x850243b7580) {}>"]
         [4226, "<Struct 'MovieSceneEvent' (0x850243bc4e0) {}>"]
         [4246, "<Struct 'MovieSceneEvent' (0x850243bd020) {}>"]
         [4261, "<Struct 'MovieSceneEvent' (0x850243b2a80) {}>"]
         [4265, "<Struct 'MovieSceneEvent' (0x850243ba1e0) {}>"]
         [4269, "<Struct 'MovieSceneEvent' (0x850243b0f00) {}>"]
         [4275, "<Struct 'MovieSceneEvent' (0x850243b1220) {}>"]
         [4281, "<Struct 'MovieSceneEvent' (0x850243b6720) {}>"]
         [4302, "<Struct 'MovieSceneEvent' (0x850243bc440) {}>"]
         [4311, "<Struct 'MovieSceneEvent' (0x850243bdac0) {}>"]
         [4627, "<Struct 'MovieSceneEvent' (0x850243b2760) {}>"]
         [4770, "<Struct 'MovieSceneEvent' (0x850243bcb20) {}>"]
        ]
       }]
     }]
   }
  ]
 }
 PostProcessVolume {
  id=#2
  possessed_class=PostProcessVolume
  tracks [
   {type=MovieSceneSpawnTrack, sections [{type=MovieSceneSpawnSection, channels [None {type=MovieSceneScriptingBoolChannel, default=true}]}]}
   {
    type=MovieScene3DTransformTrack
    sections [{
      type=MovieScene3DTransformSection
      channels [
       Location.X {type=MovieSceneScriptingDoubleChannel, default=0}
       Location.Y {type=MovieSceneScriptingDoubleChannel, default=0}
       Location.Z {type=MovieSceneScriptingDoubleChannel, default=0}
       Rotation.X {type=MovieSceneScriptingDoubleChannel, default=0}
       Rotation.Y {type=MovieSceneScriptingDoubleChannel, default=0}
       Rotation.Z {type=MovieSceneScriptingDoubleChannel, default=0}
       Scale.X {type=MovieSceneScriptingDoubleChannel, default=1}
       Scale.Y {type=MovieSceneScriptingDoubleChannel, default=1}
       Scale.Z {type=MovieSceneScriptingDoubleChannel, default=1}
      ]
     }]
   }
   {
    type=MovieSceneFloatTrack
    display_name="Exposure Compensation (Settings)"
    sections [{
      type=MovieSceneFloatSection
      channels [None {
        type=MovieSceneScriptingFloatChannel
        default=0.3
        keys [[4310,0.3],[4314,0.4],[4339,0.3],[4625,0.3],[4633,0.4],[4654,0.3],[4756,0.3],[4773,0.4],[4791,0.4],[4870,0.3]]
       }]
     }]
   }
   {
    type=MovieSceneColorTrack
    display_name="Scene Color Tint (Settings)"
    sections [{
      type=MovieSceneColorSection
      channels [
       Color.R {type=MovieSceneScriptingFloatChannel, default=1, const [1,7]}
       Color.G {
        type=MovieSceneScriptingFloatChannel
        default=1
        keys [[4310,1],[4314,0.9183],[4339,1],[4625,1],[4633,0.9183],[4654,1],[4756,1],[4773,0.9379],[4791,0.9379],[4870,1]]
       }
       Color.B {
        type=MovieSceneScriptingFloatChannel
        default=1
        keys [[4310,1],[4314,0.6806],[4339,1],[4625,1],[4633,0.6806],[4654,1],[4756,1],[4773,0.7569],[4791,0.7569],[4870,1]]
       }
       Color.A {type=MovieSceneScriptingFloatChannel, default=1, const [1,3]}
      ]
     }]
   }
   {
    type=MovieSceneFloatTrack
    display_name="Vignette Intensity (Settings)"
    sections [{
      type=MovieSceneFloatSection
      channels [None {type=MovieSceneScriptingFloatChannel, default=0.4, keys [[4756,0],[4773,0.33],[4791,0.33],[4870,0]]}]
     }]
   }
   {
    type=MovieSceneFloatTrack
    display_name="Intensity (Settings)"
    sections [{type=MovieSceneFloatSection, channels [None {type=MovieSceneScriptingFloatChannel, default=0, keys [[4756,0],[4770,0.1189],[4870,0]]}]}]
   }
  ]
 }
]
display_rate {denominator=1, numerator=60}
tick_resolution {denominator=1, numerator=24000}
tracks []
