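# Level Sequence for the GrillMoon mmp1000 segment that times the appearance of a countdown mask timer.
# A visibility track on the possessed GrillMoon_Countdown_BPmaskTimer actor keys it visible, then hidden.
# A "Mask Timer" data layer track has two back-to-back sections, switching at the same frame the actor hides.
# A sequence-level event track fires two triggers after that switch, with handlers not shown here.
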LevelSequence /GrillMoon/Sequence/mmp/mmp1000/GrillMoon_mmp1000_gam
bindings [GrillMoon_Countdown_BPmaskTimer {
  id=#1
  possessed_class=GrillMoon_Countdown_BPmaskTimer_C
  tracks [{
    type=MovieSceneVisibilityTrack
    sections [{type=MovieSceneVisibilitySection, channels [None {type=MovieSceneScriptingBoolChannel, default=true, keys [[3072,true],[3600,false]]}]}]
   }]
 }]
display_rate {denominator=1, numerator=60}
tick_resolution {denominator=1, numerator=24000}
tracks [
 {
  type=MovieSceneEventTrack
  sections [{
    type=MovieSceneEventTriggerSection
    channels [None {
      type=MovieSceneScriptingEventChannel
      keys [[4214, "<Struct 'MovieSceneEvent' (0x850243bf780) {}>"], [6582, "<Struct 'MovieSceneEvent' (0x850243b30c0) {}>"]]
     }]
   }]
 }
 {
  type=MovieSceneDataLayerTrack
  display_name=Data Layer - Mask Timer
  sections [{type=MovieSceneDataLayerSection, end_frame=3600, start_frame=73}, {type=MovieSceneDataLayerSection, end_frame=15508, start_frame=3600}]
 }
]
