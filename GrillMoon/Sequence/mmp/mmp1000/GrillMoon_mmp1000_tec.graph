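# GrillMoon mmp1000 Level Sequence that animates mask material effects through the MPC_GrillMoon parameter collection, with no object bindings.
# One material parameter collection track holds two sections: random WPO strength alone, and noise, air glow, cracks, overlay emissive and fresnel opacities.
# WPO and air glow fade early, then cracks replace noise as overlay emissive pulses.
# Cracks and fresnel later fade out together.
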LevelSequence /GrillMoon/Sequence/mmp/mmp1000/GrillMoon_mmp1000_tec
bindings []
display_rate {denominator=1, numerator=60}
tick_resolution {denominator=1, numerator=24000}
tracks [{
  type=MovieSceneMaterialParameterCollectionTrack
  display_name=MPC_GrillMoon
  sections [
   {
    type=MovieSceneParameterSection
    channels [Mask_RandomWPO_Strength {type=MovieSceneScriptingFloatChannel, default=0, keys [[3387,1],[3543,0],[4210,0]]}]
   }
   {
    type=MovieSceneParameterSection
    end_frame=12085
    start_frame=0
    channels [
     Mask_Noise_Opacity {type=MovieSceneScriptingFloatChannel, default=0, keys [[3924,1],[4210,0]]}
     Mask_Air_Glow_Opacity {type=MovieSceneScriptingFloatChannel, default=0, keys [[3387,1],[3543,0]]}
     Mask_Cracks_Opacity {type=MovieSceneScriptingFloatChannel, default=0, thinned_from=5, keys [[4206,0],[4210,1],[6165,1],[6239,0]]}
     Mask_Overlay_Emissive_Opacity {type=MovieSceneScriptingFloatChannel, default=0.1, keys [[3387,0],[3639,0.1],[4113,0],[4341,1],[4522,0],[4613,0.5],[4639,1],[4766,0]]}
     Mask_Fresnel_Opacity {type=MovieSceneScriptingFloatChannel, default=0, keys [[6165,1],[6239,0]]}
    ]
   }
  ]
 }]
